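# MF_CameraDistanceFade is a material function in the OilSpill environment folder that outputs a fade value from 0 near the camera to 1 farther away.
# It feeds pixel depth and a zero constant into a SphereMask, using it as a one-dimensional falloff, then inverts the result.
# The Distance input sets the mask radius and has a usable default, while low hardness gives a soft, gradual transition.
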MaterialFunction /Clyde/Environments/OilSpill/MF_CameraDistanceFade
graph {
 properties {libraryCategoriesText [Misc,Effects]}
 nodes [
  n1 = PixelDepth()
  n2 = Constant()
  n3 = FunctionInput(Preview=Constant() {r=256}) {bUsePreviewValueAsDefault=true, desc=(0,1), description="Distance to Object (0,1)", inputName=Distance, inputType=FunctionInput_Scalar}
  FunctionOutput(None=OneMinus(None=SphereMask(A=@n1.R, B=@n2, Radius=@n3) {hardnessPercent=10}))
 ]
}
